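; PCG primitive graph that places a supplied list of assets on a grid, based on an origin position and padding ratio.
; A core process subgraph receives the asset array, origin and padding and does the placement work.
; Its result feeds a prepare-outputs subgraph, overridable through a graph parameter that defaults to the shared template, before reaching the graph output.
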(graph Place_Assets_OnAGrid
  (params (Assets :type (Array (SoftObject /Script/CoreUObject.Object))) (originPosition :type Vector) (paddingRatio :type Double :default 1.5) (PrepareOutputsGraph :type (SoftObject /Script/CoreUObject.Object) :default /PCGPrimitives/Primitives/_Template_/Primitives_PrepareOutputs.Primitives_PrepareOutputs))
  ;[0 0 0 0.5 32] Core Process
  (bind Subgraph_3 (Subgraph :Graph /PCGPrimitives/Primitives/Place/Subgraphs/Place_Assets_OnAGrid_CoreProcess.Place_Assets_OnAGrid_CoreProcess :Assets Assets :origin_position originPosition :padding_ratio paddingRatio))
  ;[0 0 0 0.5 32] Prepare Output
  (bind Subgraph_0 (Subgraph :Graph /PCGPrimitives/Primitives/_Template_/Primitives_PrepareOutputs.Primitives_PrepareOutputs :SubgraphOverride-param PrepareOutputsGraph))
  (connect Subgraph_3 Out Subgraph_0 In)
  (output Out Subgraph_0))
