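; PCG primitive graph that places points along a spline, with options for spacing or fixed count, existing control points, facing, offsets and curve following or fitting.
; It tags the input as "Input:0", then runs inputs management, which by default uses a force-spline processing graph and a template fallback.
; The result feeds the Place_OnSpline core process, then a point output-preparation subgraph.
; Graph parameters expose the processing, fallback, core and output stages as swappable subgraphs.
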(graph Place_OnSpline
  (params (InputsProcessingGraph :type (SoftObject /Script/PCG.PCGGraphInterface) :default /PCGPrimitives/Primitives/Shared/InputsProcessing/InputsProcessing_ForceSpline.InputsProcessing_ForceSpline) (InputsFallbackGraph :type (SoftObject /Script/PCG.PCGGraphInterface) :default /PCGPrimitives/Primitives/_Template_/Primitives_InputsFallback.Primitives_InputsFallback) (CoreProcessGraph :type (SoftObject /Script/PCG.PCGGraphInterface) :default /PCGPrimitives/Primitives/Place/Subgraphs/Place_OnSpline_CoreProcess.Place_OnSpline_CoreProcess) (PrepareOutputsGraph :type (SoftObject /Script/PCG.PCGGraphInterface) :default /PCGPrimitives/Primitives/Shared/PrepareOutputs/PrepareOutputs_Points.PrepareOutputs_Points) (UseExistingControlPoints :type Bool) (Spacing :type Double :default 1.5e+02) (UseFixedCount :type Bool) (Count :type Int) (Facing :type (Enum /PCGPrimitives/Primitives/Place/Subgraphs/EPCG_Facing.EPCG_Facing)) (MeshForwardAxis :type (Enum /Script/CoreUObject.EAxis) :default X) (FollowCurve :type Bool :default true) (Offset :type Vector) (splineStartOffset :type Double) (splineEndOffset :type Double) (FitToCurve :type Bool))
  (inputs (PrimaryInput :type Any))
  (bind AddTags_0 (Add%20Tags :TagsToAdd "Input:0"))
  ;[0 0 0 0.5 32] Inputs
  (bind Subgraph_0 (Subgraph :Graph /PCGPrimitives/Primitives/_Template_/Subgraphs/Primitives_InputsManagement.Primitives_InputsManagement :InputsFallbackGraph InputsFallbackGraph :InputsProcessingGraph InputsProcessingGraph))
  ;[0 0 0 0.5 32] Core Process
  (bind Subgraph_3 (Subgraph :Graph /PCGPrimitives/Primitives/Place/Subgraphs/Place_OnSpline_CoreProcess.Place_OnSpline_CoreProcess :Count Count :EndOffset SplineEndOffset :Facing Facing :FitToCurve FitToCurve :FollowCurve FollowCurve :Offset Offset :Spacing Spacing :StartOffset SplineStartOffset :SubgraphOverride-param CoreProcessGraph :UseExistingControlPoints UseExistingControlPoints :UseFixedCount UseFixedCount))
  ;[0 0 0 0.5 32] Prepare Output
  (bind Subgraph_4 (Subgraph :Graph /PCGPrimitives/Primitives/Shared/PrepareOutputs/PrepareOutputs_Points.PrepareOutputs_Points :SubgraphOverride-param PrepareOutputsGraph))
  (connect DefaultInputNode PrimaryInput AddTags_0 In)
  (connect AddTags_0 Out Subgraph_0 In)
  (connect Subgraph_0 Out Subgraph_3 In)
  (connect Subgraph_3 Out Subgraph_4 In)
  (output Out Subgraph_4))
