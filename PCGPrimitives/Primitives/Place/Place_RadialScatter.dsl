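; Place_RadialScatter is a PCG primitive graph that scatters points radially in rows, set by start radius, radius increment, row count and spacing along the arc.
; It tags the input, then chains input management (point conversion, empty fallback by default), a core scatter subgraph and output preparation.
; Parameters forward scale, height variation, seed and 90-degree rotation to the core.
; Input processing, fallback, core and output stages are swappable through graph parameters.
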(graph Place_RadialScatter
  (params (InputsProcessingGraph :type (SoftObject /Script/PCG.PCGGraphInterface) :default /PCGPrimitives/Primitives/Shared/InputsProcessing/InputsProcessing_ConvertToPoints.InputsProcessing_ConvertToPoints) (InputsFallbackGraph :type (SoftObject /Script/PCG.PCGGraphInterface) :default /PCGPrimitives/Primitives/Shared/InputsFallback/InputsFallback_Empty.InputsFallback_Empty) (CoreProcessGraph :type (SoftObject /Script/PCG.PCGGraphInterface) :default /PCGPrimitives/Primitives/Place/Subgraphs/Place_Radial_scatter_CoreProcess.Place_Radial_scatter_CoreProcess) (PrepareOutputsGraph :type (SoftObject /Script/PCG.PCGGraphInterface) :default /PCGPrimitives/Primitives/_Template_/Primitives_PrepareOutputs.Primitives_PrepareOutputs) (IncrementRadiusDistance :type Double :default 2e+03) (StartRadius :type Double :default 3e+03) (PointEvery_X_Cm_On_Arc :type Double :default 1e+03) (NumberOfRows :type Int :default 10) (ScaleMin :type Double :default 0.8) (ScaleMax :type Double :default 1.2) (heightVariation :type Double) (Seed :type Int) (Random90Rotation :type Bool))
  (inputs (PrimaryInput :type Any))
  (bind AddTags_1 (Add%20Tags :TagsToAdd "Input:0"))
  ;[0 0 0 0.5 32] Inputs
  (bind Subgraph_0 (Subgraph :Graph /PCGPrimitives/Primitives/_Template_/Subgraphs/Primitives_InputsManagement.Primitives_InputsManagement :InputsFallbackGraph InputsFallbackGraph :InputsProcessingGraph InputsProcessingGraph))
  ;[0 0 0 0.5 32] Core Process
  (bind Subgraph_3 (Subgraph :Graph /PCGPrimitives/Primitives/Place/Subgraphs/Place_Radial_scatter_CoreProcess.Place_Radial_scatter_CoreProcess :Height_variation HeightVariation :increment_Radius_distance IncrementRadiusDistance :Number_of_Rows NumberOfRows :Point_every_x_cm_on_arc PointEvery_X_Cm_On_Arc :Random_90_rotation Random90Rotation :Scale_max ScaleMax :Scale_min ScaleMin :Seed-param Seed :Start_Radius StartRadius :SubgraphOverride-param CoreProcessGraph))
  ;[0 0 0 0.5 32] Prepare Output
  (bind Subgraph_4 (Subgraph :Graph /PCGPrimitives/Primitives/_Template_/Primitives_PrepareOutputs.Primitives_PrepareOutputs :SubgraphOverride-param PrepareOutputsGraph))
  (connect DefaultInputNode PrimaryInput AddTags_1 In)
  (connect AddTags_1 Out Subgraph_0 In)
  (connect Subgraph_0 Out Subgraph_3 In)
  (connect Subgraph_3 Out Subgraph_4 In)
  (output Out Subgraph_4))
